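; Gameplay ability for the ProtoSwarm medium AI's ranged attack, lobbing an arcing projectile at its blackboard goal actor beyond a minimum range.
; After a short delay it randomly roars or plays the attack montage, whose notify fires the projectile.
; Aim leads the target by its velocity, solves a custom arc, and spawns a landing-location actor linked to the projectile.
; Takeover-controlled pawns skip the range check and roar, firing straight ahead.
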(event Ability|EventActivateAbility
  (Utilities|FlowControl|Delay 1.0)
  (if (GameplayTags|HasMatchingGameplayTag (Ability|GetActivatingPawn) "(TagName=\"Athena.Gameplay.Status.IsControlledByTakeOverAIPawn\")")
    (CallFunction|DoRangeAttack)
    (elif (<= (Math|Random|RandomFloatinRange 0.0 1.0) (ScalableFloat|GetValueatLevelwithContextParam (Ability|GetOwningActorfromActorInfo) (Variables|Default|GetChanceToRoarInsteadOfAttack)))
      (CallFunction|DoRoar)
      (else
        (CallFunction|DoRangeAttack)))))

(event Custom|DoRangeAttack
  (bind (_montageinfo_montagetoplay _montageinfo_animplayrate _montageinfo_animrootmotiontranslationscale _montageinfo_montagesectiontoplay _montageinfo_overridesection) (Utilities|Struct|BreakFortGameplayAbilityMontageInfo (Variables|GetMontageInfo_MontageToPlay)))
  (Ability|Tasks|PlayMontageAndWaitNotify "None" _montageinfo_montagetoplay _montageinfo_animplayrate _montageinfo_overridesection true _montageinfo_animrootmotiontranslationscale
    (:then
      (Ability|ExecuteGameplayCueOnOwner (Variables|Default|GetGC_RangeAttack)))
    (:OnCompleted
      _)
    (:OnInterrupted
      _)
    (:OnCancelled
      _)
    (:OnNotifyBegin
      (CallFunction|FireProjectile)))
  (Ability|CommitAbility)
  (Ability|EndAbility))

(event Custom|FireProjectile
  (bind (_targetlocation _result) (CallFunction|GetTargetLocation))
  (if _result
    (bind _output_get (Variables|InternalData|SetTargetLocation _targetlocation))
    (bind _outlaunchvelocity (Game|SuggestProjectileVelocityCustomArc (Ability|GetFireLocation) _output_get (ScalableFloat|GetValueatLevel (Variables|Default|GetProjectileGravityOverride)) (ScalableFloat|GetValueatLevel (Variables|Default|GetProjectileArcParam))))
    (bind _returnvalue (Projectile|SpawnProjectile :ProjectileClass (Variables|Ability|GetProjectileClass) :RequestedBy (Ability|GetActivatingPawn) :SpawnLocation (Ability|GetFireLocation) :SpawnRotation "0, 0, 0" :SpawnDirection (Math|Conversions|RotationFromXVector _outlaunchvelocity) :EffectContainerSpecToApplyOnExplode (Ability|MakeGameplayEffectContainerSpec "(TagName=\"Targeting.ApplicationTag.0\")") :bUseDefaultPhysics false :InitialSpeed (Math|Vector|VectorLength _outlaunchvelocity) :GravityScale (ScalableFloat|GetValueatLevel (Variables|Default|GetProjectileGravityScale)) :ChargePercent 0.0))
    (bind _asprj_tidal_crane_medium_ranged (Utilities|Casting|CastToPrj_TidalCrane_Medium_Ranged _returnvalue))
    (Variables|InternalData|SetRangedProjectile _asprj_tidal_crane_medium_ranged)
    (bind _returnvalue_2 (Game|SpawnActorfromClass "/ProtoSwarm/Blueprints/BGA_TidalCrane_Lobber_ProjectileLocation.BGA_TidalCrane_Lobber_ProjectileLocation_C" (Math|Transform|MakeTransform (Variables|InternalData|GetTargetLocation)) "AlwaysSpawn" "MultiplyWithRoot" (Ability|GetActivatingPawn) 0 (Ability|GetActivatingPawn)))
    (Variables|/ProtoSwarm/Abilities/Medium/Prj_TidalCrane_Medium_Ranged.Prj_TidalCrane_Medium_Ranged_C|SetTargetProjection _returnvalue_2 (Variables|InternalData|GetRangedProjectile))
    (Ability|ExecuteGameplayCueOnOwner (Variables|Default|GetGC_RangeAttacking))
    (else
      (Ability|EndAbility))))

(event Custom|DoRoar
  (bind (_roarmontageinfo_montagetoplay _roarmontageinfo_animplayrate _roarmontageinfo_animrootmotiontranslationscale _roarmontageinfo_montagesectiontoplay _roarmontageinfo_overridesection) (Utilities|Struct|BreakFortGameplayAbilityMontageInfo (Variables|GetRoarMontageInfo_MontageToPlay)))
  (Ability|Tasks|PlayMontageAndWaitNotify "None" _roarmontageinfo_montagetoplay _roarmontageinfo_animplayrate _roarmontageinfo_overridesection true _roarmontageinfo_animrootmotiontranslationscale
    (:OnCompleted
      _)
    (:OnInterrupted
      _)
    (:OnCancelled
      _))
  (Ability|CommitAbility)
  (Ability|EndAbility))

(fn GetGoalActor
  (return (AI|Components|Blackboard|GetValueasObject (AI|GetBlackboard (AI|GetAIController (Ability|GetAvatarActorfromActorInfo))) (Variables|Default|GetGolaActorKeyName))))

(fn GetTargetLocation
  (if (GameplayTags|HasMatchingGameplayTag (Ability|GetActivatingPawn) "(TagName=\"Athena.Gameplay.Status.IsControlledByTakeOverAIPawn\")")
    ;[0.1303 0.1406 0] Cheat AITakeOver
    (return (+ (Transformation|GetActorLocation (Ability|GetActivatingPawn)) (* (Transformation|GetActorForwardVector (Ability|GetActivatingPawn)) 2e+03)) true)
    (else
      (bind _asactor (Utilities|Casting|CastToActor (CallFunction|GetGoalActor))
        (:then
          (return (+ (Transformation|GetActorLocation _asactor) (Transformation|GetVelocity _asactor)) true))
        (:CastFailed
          (return "0, 0, 0" false))))))

(fn CanActivateAbility
  (params (ActorInfo :type Gameplay%20Ability%20Actor%20Info%20Structure) (Handle :type Gameplay%20Ability%20Spec%20Handle%20Structure))
  (bind _returnvalue (Ability|GetAvatarActorfromActorInfo))
  (if (GameplayTags|HasMatchingGameplayTag (Ability|GetActivatingPawn) "(TagName=\"Athena.Gameplay.Status.IsControlledByTakeOverAIPawn\")")
    (return 0 true)
    (else
      (bind _asfort_pawn (Utilities|Casting|CastToFortPawn _returnvalue)
        (:then
          (bind _targetactor (Class|BPFLAICommonFunctions|GetTargetActorfromBlackboard _asfort_pawn "GoalActor"))
          (Utilities|IsValid _targetactor
            (:Is%20Valid
              (return 0 (> (Math|Vector|Distance(Vector) (Transformation|GetActorLocation _targetactor) (Transformation|GetActorLocation _returnvalue)) (ScalableFloat|GetValueatLevel (Variables|Default|GetMinimumRange)))))
            (:Is%20Not%20Valid
              _)))
        (:CastFailed
          _))
      (return 0 false))))
