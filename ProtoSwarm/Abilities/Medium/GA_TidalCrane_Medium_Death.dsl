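; Gameplay ability for the death of the medium Tidal Crane, triggered by a gameplay event and made non-cancelable.
; If the event's target tags include the despawn tag, it skips the montage, waits out the death FX duration, then ends.
; Otherwise it plays the death montage, runs explode-tagged target selection to apply a gameplay effect container, and ends when the montage completes or is cancelled.
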(event Ability|EventActivateAbilityFromEvent
  (params (EventData :type Gameplay%20Event%20Data%20Structure%20%28by%20ref%29))
  (bind (_eventtag _instigator _target _optionalobject _optionalobject2 _contexthandle _instigatortags _targettags) (Utilities|Struct|BreakGameplayEventData EventData))
  (bind (_montageinfo_montagetoplay _montageinfo_animplayrate _montageinfo_animrootmotiontranslationscale _montageinfo_montagesectiontoplay _montageinfo_overridesection) (Utilities|Struct|BreakFortGameplayAbilityMontageInfo (Variables|GetMontageInfo_MontageToPlay)))
  (Ability|SetCanbeCanceled false)
  (if (GameplayTags|HasTag _targettags "(TagName=\"Ability.TidalCrane.Despawn\")")
    (Utilities|FlowControl|Delay (Variables|Default|GetDurationToPlayDeathFXIfSkippingMontage)) ; Duration despawn FX should play before we hide pawn
    (Ability|EndAbility)
    (else
      (Ability|Tasks|PlayMontageAndWaitNotify "None" _montageinfo_montagetoplay _montageinfo_animplayrate _montageinfo_overridesection true _montageinfo_animrootmotiontranslationscale
        (:then
          (bind (_asynctaskproxy _targetdata _applicationtag) (Ability|Tasks|DoTargetSelection "None" "(TagName=\"Gameplay.Application.Explode\")")
            (:Targeted
              (Ability|ApplyGameplayEffectContainer _targetdata _applicationtag))))
        (:OnCompleted
          _)
        (:OnCancelled
          _))
      (Ability|EndAbility))))
