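; Gameplay ability for the medium TidalCrane ProtoSwarm AI that triggers a burrow reposition after heavy damage.
; If enabled, it waits briefly on activation, records pawn health, binds OnDamaged and listens for the Reposition tag to timestamp burrows.
; When health lost and time since last burrow exceed tunable thresholds, and no takeover controls the pawn, it resets the health baseline and sets the ThreatsInProximity blackboard key.
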(event Ability|EventActivateAbility
  (Utilities|FlowControl|Delay 0.2)
  (if (ScalableFloat|GetValueasBool (Variables|Default|GetRepositionFromHealthTresholdEnabled))
    (bind _asfort_aipawn (Utilities|Casting|CastToFortAIPawn (Ability|GetAvatarActorfromActorInfo))
      (:then
        (bind _output_get (Variables|InternalData|SetAIPawn _asfort_aipawn))
        (CallFunction|ListentoBurrowTag)
        (EventDispatchers|BindEventtoOnDamaged _output_get (EventDispatchers|CreateEvent :!event Custom|AIPawnWasDamaged))
        (Variables|InternalData|SetHealthAtLastReposition (FortPawn|GetHealth _output_get)))
      (:CastFailed
        _))
    (else
      _))
  (Ability|EndAbility))

(event Custom|ListenToBurrowTag
  ; Track the last time the AI Burrowed / Repositionned
  (Ability|Tasks|WaitGameplayTagAdd "(TagName=\"Ability.ProtoSwarm.NonAttack.Reposition\")" (Variables|InternalData|GetAIPawn)
    (:Added
      (Variables|InternalData|SetLastBurrowTime (Utilities|Time|GetGameTimeinSeconds)))))

(event Custom|AIPawnWasDamaged
  (params (DamagedActor :type Actor%20Object%20Reference) (Damage :type Float%20%28single-precision%29) (InstigatedBy :type Controller%20Object%20Reference) (DamageCauser :type Actor%20Object%20Reference) (HitLocation :type Vector) (FHitComponent :type Primitive%20Component%20Object%20Reference) (BoneName :type Name) (Momentum :type Vector))
  ; When Damaged, Look when was the last time we burrowed and if we received a certain amount of damage and if so, try to burrow!
  (if (Math|Boolean|ANDBoolean (> (- (Variables|InternalData|GetHealthAtLastReposition) (FortPawn|GetHealth (Variables|InternalData|GetAIPawn))) (ScalableFloat|GetValueatLevel (Variables|Default|GetForceRepositionHealthThresholds))) (> (- (Utilities|Time|GetGameTimeinSeconds) (Variables|InternalData|GetLastBurrowTime)) (ScalableFloat|GetValueatLevel (Variables|Default|GetMinimumTimeBetweenRepositions))) (not (GameplayTags|HasMatchingGameplayTag (Variables|InternalData|GetAIPawn) "(TagName=\"Athena.Gameplay.Status.IsControlledByTakeOverAIPawn\")")))
    (Variables|InternalData|SetHealthAtLastReposition (FortPawn|GetHealth (Variables|InternalData|GetAIPawn)))
    (AI|Components|Blackboard|SetValueasBool (AI|GetBlackboard (Pawn|GetController (Variables|InternalData|GetAIPawn))) (Variables|Default|GetKey_ThreatsInProximity) true)))
