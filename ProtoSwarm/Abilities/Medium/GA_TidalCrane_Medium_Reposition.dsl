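; Gameplay ability that makes the medium Tidal Crane AI burrow, relocate and erupt at a new spot.
; An EQS query, swapped under AI takeover, picks the location; nearby music encounter volumes pass their tag to the pawn.
; Burrow montage notifies drive hiding, a tell actor, teleporting to face the target, and a delayed unburrow.
; Unburrowing applies targeted effects and knocks back pawns and vehicles; ability end restores visibility and rechecks threats.
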(event Ability|EventActivateAbility
  (Ability|CommitAbility)
  (bind _asfort_pawn (Utilities|Casting|CastToFortPawn (Ability|GetAvatarActorfromActorInfo))
    (:then
      (bind _output_get (Variables|Internal|SetAIPawn _asfort_pawn))
      (Variables|Default|SetAIController (Pawn|GetController _output_get))
      (CallFunction|FindUnburrowLocation))
    (:CastFailed
      (Ability|EndAbility))))

(event Custom|FindUnburrowLocation
  (bind _returnvalue (AI|EQS|RunEQSQuery (select (GameplayTags|HasMatchingGameplayTag (Ability|GetActivatingPawn) "(TagName=\"Athena.Gameplay.Status.IsControlledByTakeOverAIPawn\")") (Variables|Default|GetUnBurrowEQS_AITakeOver) (Variables|Default|GetUnBurrowEQS)) (Variables|Internal|GetAIPawn)))
  (AI|EQS|BindEventtoOnQueryFinished _returnvalue (EventDispatchers|CreateEvent :!event Custom|UnburrowEQSFinished)))

(event Custom|UnburrowEQSFinished
  (params (QueryInstance :type EQS%20Query%20Instance%20Object%20Reference) (QueryStatus :type EEnvQueryStatus%20Enum :default "Processing"))
  (bind _aipawn (Variables|Internal|GetAIPawn))
  (bind _capsulecomponent (Variables|/Script/Engine.Character|GetCapsuleComponent _aipawn))
  (bind _returnvalue (Components|Capsule|GetScaledCapsuleHalfHeight _capsulecomponent))
  (if (Utilities|Enum|Equal(Enum) QueryStatus "Success")
    (bind (_resultlocations _returnvalue_2) (AI|EQS|GetQueryResultsasLocations QueryInstance))
    (if _returnvalue_2
      (CallFunction|StartRelocation (Utilities|Operators|Add (Utilities|Array|Get(acopy) _resultlocations) (Math|Vector|MakeVector 0.0 0.0 _returnvalue)))
      ; Check if moving into a music encounter volume and pass the instance tag to the pawn
      (bind (_outactors _returnvalue_3) (Collision|SphereOverlapActors (Utilities|Operators|Add (Utilities|Array|Get(acopy) _resultlocations) (Math|Vector|MakeVector 0.0 0.0 _returnvalue)) 1e+03 (Utilities|Array|MakeArray "ObjectTypeQuery2") "/Script/Engine.TriggerVolume"))
      (if _returnvalue_3
        (bind _array_element (Utilities|Array|ForEachLoopwithBreak _outactors)
          (:LoopBody
            (Utilities|IsValid _array_element
              (:Is%20Valid
                (Utilities|IsValid (Actor|GetComponentbyClass _array_element "/MusicEncounter/Components/BP_MusicEncounterComponent.BP_MusicEncounterComponent_C")
                  (:Is%20Valid
                    (MusicEncounter|SetMusicEncounterInstanceTag(Message) (Variables|Internal|GetAIPawn) (Variables|/MusicEncounter/Components/BP_MusicEncounterComponent.BP_MusicEncounterComponent_C|GetMusic%20Encounter%20Instance%20Tag (Actor|GetComponentbyClass _array_element "/MusicEncounter/Components/BP_MusicEncounterComponent.BP_MusicEncounterComponent_C"))))))))))
      (else
        _))
    (else
      _))
  (Ability|EndAbility))

(event Custom|StartRelocation
  (params (UnburrowLocation :type Vector))
  (bind (_montageinfo_montagetoplay _montageinfo_animplayrate _montageinfo_animrootmotiontranslationscale _montageinfo_montagesectiontoplay _montageinfo_overridesection) (Utilities|Struct|BreakFortGameplayAbilityMontageInfo (Variables|GetMontageInfo_MontageToPlay)))
  (bind _aipawn (Variables|Internal|GetAIPawn))
  ; Start Burrow
  (bind (_asynctaskproxy _notifytag) (Ability|Tasks|PlayMontageAndWaitNotify "None" _montageinfo_montagetoplay _montageinfo_animplayrate _montageinfo_overridesection true _montageinfo_animrootmotiontranslationscale)
    (:then
      (Ability|AddGameplayCueToOwner (Variables|Default|GetGC_Repositioning)))
    (:OnNotifyBegin
      (if (== _notifytag "(TagName=\"Event.TidalCrane.Lobber.Burrowing\")")
        (Ability|ExecuteGameplayCueOnOwner (Variables|Default|GetGC_Burrow))
        (else
          (Rendering|SetActorHiddenInGame _aipawn true)
          (bind _returnvalue (Game|SpawnActorfromClass "/ProtoSwarm/Blueprints/Medium/BP_TidalCrane_Medium_UnburrowTell.BP_TidalCrane_Medium_UnburrowTell_C" (CallFunction|GetTransformonFloor _aipawn) "AlwaysSpawn" "MultiplyWithRoot" _aipawn 0 _aipawn))
          (bind _output_get (Variables|Default|SetBurrowTellActor _returnvalue))
          (Class|BPTidalCraneMediumUnburrowTell|SetUnborrowLocation _output_get (+ (* (Math|Vector|Normalize (- UnburrowLocation (Utilities|Operators|Subtract (Transformation|GetActorLocation (Variables|Internal|GetAIPawn)) (Math|Vector|MakeVector 0.0 0.0 (Components|Capsule|GetScaledCapsuleHalfHeight (Variables|/Script/Engine.Character|GetCapsuleComponent (Variables|Internal|GetAIPawn))))))) 2e+03) (Utilities|Operators|Subtract (Transformation|GetActorLocation (Variables|Internal|GetAIPawn)) (Math|Vector|MakeVector 0.0 0.0 (Components|Capsule|GetScaledCapsuleHalfHeight (Variables|/Script/Engine.Character|GetCapsuleComponent (Variables|Internal|GetAIPawn)))))))
          (Utilities|IsValid (CallFunction|GetUnburrowTarget)
            (:Is%20Valid
              _)
            (:Is%20Not%20Valid
              _))
          (bind _data (Utilities|FlowControl|MergeExecutionwithData (Math|Rotator|FindLookatRotation UnburrowLocation (Transformation|GetActorLocation (CallFunction|GetUnburrowTarget))) (Math|Rotator|MakeRotator 0.0))
            (:Then
              (Transformation|SetActorLocationAndRotation (Variables|Internal|GetAIPawn) UnburrowLocation _data false true)
              (FortPawn|SetPawnandControlRotation (Variables|Internal|GetAIPawn) _data)
              (Utilities|FlowControl|Delay (ScalableFloat|GetValueatLevel (Variables|Default|GetBurrowDuration)))
              (CallFunction|StartUnburrow))))))))

(event Custom|StartUnburrow
  ; Start Unburrow
  (Ability|Tasks|PlayMontageAndWaitNotify "None" (Variables|Default|GetUnburrowAnimation)
    (:then
      (Ability|ExecuteGameplayCueOnOwner (Variables|Default|GetGC_Unburrow))
      (Rendering|SetActorHiddenInGame (Variables|Internal|GetAIPawn) false)
      (bind (_asynctaskproxy _targetdata _applicationtag) (Ability|Tasks|DoTargetSelection "None" "(TagName=\"Targeting.ApplicationTag.0\")" true)
        (:Targeted
          (CallFunction|ApplyDamageandKnockback _targetdata _applicationtag))))
    (:OnCompleted
      _)
    (:OnInterrupted
      _)
    (:OnCancelled
      _))
  (Ability|EndAbility))

(event Custom|ApplyDamageAndKnockback
  (params (TargetData :type Gameplay%20Ability%20Target%20Data%20Handle%20Structure) (ApplicationTag :type Gameplay%20Tag%20Structure))
  ; KNockback Stuff
  (Ability|ApplyGameplayEffectContainer TargetData ApplicationTag)
  (for _array_element (Ability|TargetData|GetAllActorsfromTargetData TargetData)
    (bind _asfort_pawn (Utilities|Casting|CastToFortPawn _array_element)
      (:then
        (Default|KnockbackPawns _asfort_pawn))
      (:CastFailed
        (bind _asfort_athena_vehicle (Utilities|Casting|CastToFortAthenaVehicle _array_element))
        (Default|KnockbackVehicles _asfort_athena_vehicle)))))

(event Ability|EventOnEndAbility
  (params (bWasCancelled :type Boolean))
  (bind _aipawn (Variables|Internal|GetAIPawn))
  (Utilities|IsValid _aipawn
    (:Is%20Valid
      (Rendering|SetActorHiddenInGame _aipawn false)
      (Ability|RemoveGameplayCueFromOwner (Variables|Default|GetGC_Burrow))
      (bind _aspawn_tidal_crane_medium (Utilities|Casting|CastToPawn_TidalCrane_Medium _aipawn))
      (Class|PawnTidalCraneMedium|CheckforValidThreat _aspawn_tidal_crane_medium))))

(fn GetTransformOnFloor
  (params (AIPawn :type Fort%20Pawn%20Object%20Reference))
  (bind (_location _rotation _scale) (Math|Transform|BreakTransform (Transformation|GetActorTransform AIPawn)))
  (bind (_location_x _location_y _location_z) (Math|Vector|BreakVector _location))
  (return (Math|Transform|MakeTransform (Math|Vector|MakeVector _location_x _location_y (- _location_z (Components|Capsule|GetScaledCapsuleHalfHeight (Variables|/Script/Engine.Character|GetCapsuleComponent AIPawn)))) _rotation _scale)))

(fn GetUnburrowTarget
  (return (Ability|GetAIAbilityTarget)))

(fn KnockbackPawns
  (params (Pawn :type Fort%20Pawn%20Object%20Reference))
  (bind (_returnvalue_x _returnvalue_y) (Math|Vector|BreakVector (Math|Vector|Normalize (- (Transformation|GetActorLocation Pawn) (Transformation|GetActorLocation (Ability|GetAvatarActorfromActorInfo))))))
  (if (not (GameplayTags|HasMatchingGameplayTag Pawn "(TagName=\"Gameplay.Status.KnockbackImmunity\")"))
    (Character|LaunchCharacter Pawn (Math|Vector|MakeVector (* _returnvalue_x (ScalableFloat|GetValueatLevel (Variables|Default|GetKnockbackPawnForwardVelocity))) (* _returnvalue_y (ScalableFloat|GetValueatLevel (Variables|Default|GetKnockbackPawnForwardVelocity))) (ScalableFloat|GetValueatLevel (Variables|Default|GetKnockbackPawnHeightVelocity))) true true)))

(fn KnockbackVehicles
  (params (FortAthenaVehicle :type Fort%20Athena%20Vehicle%20Object%20Reference))
  (bind (_returnvalue_x _returnvalue_y) (Math|Vector|BreakVector (Math|Vector|Normalize (- (Transformation|GetActorLocation FortAthenaVehicle) (Transformation|GetActorLocation (Ability|GetAvatarActorfromActorInfo))))))
  (Audio|BroadcastAbilityImpactatLocation 0 0 (Math|Vector|MakeVector (* _returnvalue_x (ScalableFloat|GetValueatLevel (Variables|Default|GetKnockbackVehicleForwardVelocity))) (* _returnvalue_y (ScalableFloat|GetValueatLevel (Variables|Default|GetKnockbackVehicleForwardVelocity))) (ScalableFloat|GetValueatLevel (Variables|Default|GetKnockbackVehicleHeightVelocity))) 777.0 "None" false))
