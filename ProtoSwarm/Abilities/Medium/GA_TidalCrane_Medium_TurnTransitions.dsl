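; Gameplay ability for turn transitions of the medium TidalCrane in ProtoSwarm.
; On activation it commits, then checks whether the picked turn transition uses a montage.
; If so, a custom task plays the montage with a rotation curve from the picked transition data; otherwise the ability ends immediately.
; The montage task's completed, interrupted and cancelled callbacks and the end-ability event have no logic in this graph.
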(event Ability|EventActivateAbility
  (Ability|CommitAbility)
  (if (Variables|TurnTransition|GetbPickedTurnTransitionUsesMontage)
    (Ability|Tasks|PlayMontageWithRotCurveAndWaitNotify "None" (Variables|TurnTransition|GetPickedData)
      (:OnCompleted
        _)
      (:OnInterrupted
        _)
      (:OnCancelled
        _))
    (Ability|EndAbility)))

(event Ability|EventOnEndAbility
  (params (bWasCancelled :type Boolean)))
