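; Spawn-in gameplay ability for the medium TidalCrane swarm AI: fires a spawn cue, plays a montage and unhides the avatar.
; Target selection applies an effect container to targets and launches non-immune pawns away from the avatar.
; The vehicle branch computes a similar knockback vector but feeds it to an ability-impact broadcast instead of a launch.
; It also checks nearby trigger volumes for a music encounter component and passes its instance tag to the pawn.
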(event Ability|EventActivateAbility
  (bind (_montageinfo_montagetoplay _montageinfo_animplayrate _montageinfo_animrootmotiontranslationscale _montageinfo_montagesectiontoplay _montageinfo_overridesection) (Utilities|Struct|BreakFortGameplayAbilityMontageInfo (Variables|GetMontageInfo_MontageToPlay)))
  (Ability|CommitAbility)
  (Utilities|FlowControl|Delay 0.2)
  (Ability|ExecuteGameplayCueOnOwner (Variables|Default|GetGC_SpawnIn))
  (Ability|Tasks|PlayMontageAndWaitNotify "None" _montageinfo_montagetoplay _montageinfo_animplayrate _montageinfo_overridesection true _montageinfo_animrootmotiontranslationscale
    (:then
      (Rendering|SetActorHiddenInGame (Ability|GetAvatarActorfromActorInfo) false)
      (bind (_asynctaskproxy _targetdata _applicationtag) (Ability|Tasks|DoTargetSelection "None" "(TagName=\"Targeting.ApplicationTag.0\")" true)
        (:then
          (CallFunction|CheckForMusicEncounterVolume))
        (:Targeted
          (CallFunction|ApplyDamageandKnockback _targetdata _applicationtag))))
    (:OnCompleted
      _)
    (:OnInterrupted
      _)
    (:OnCancelled
      _))
  (Ability|EndAbility))

(event Ability|EventFailedToActivatePassiveAbility
  (params (ActorInfo :type Gameplay%20Ability%20Actor%20Info%20Structure)))

(event Custom|ApplyDamageAndKnockback
  (params (TargetData :type Gameplay%20Ability%20Target%20Data%20Handle%20Structure) (ApplicationTag :type Gameplay%20Tag%20Structure))
  (Ability|ApplyGameplayEffectContainer TargetData ApplicationTag)
  (for _array_element (Ability|TargetData|GetAllActorsfromTargetData TargetData)
    (bind _asfort_pawn (Utilities|Casting|CastToFortPawn _array_element)
      (:then
        (Default|KnockbackPawns _asfort_pawn))
      (:CastFailed
        (bind _asfort_athena_vehicle (Utilities|Casting|CastToFortAthenaVehicle _array_element))
        (Default|KnockbackVehicles _asfort_athena_vehicle)))))

(event Ability|EventOnEndAbility
  (params (bWasCancelled :type Boolean)))

(event Custom|Check%20for%20Music%20Encounter%20Volume
  ; Check if spawning into a music encounter volume and pass the instance tag to the pawn
  (bind _asfort_pawn (Utilities|Casting|CastToFortPawn (Ability|GetAvatarActorfromActorInfo)))
  (bind _output_get (Variables|Default|SetAI%20Pawn _asfort_pawn))
  (bind (_outactors _returnvalue) (Collision|SphereOverlapActors (Transformation|GetActorLocation _output_get) 1e+03 (Utilities|Array|MakeArray "ObjectTypeQuery2") "/Script/Engine.TriggerVolume"))
  (if _returnvalue
    (bind _array_element (Utilities|Array|ForEachLoopwithBreak _outactors)
      (:LoopBody
        (Utilities|IsValid _array_element
          (:Is%20Valid
            (Utilities|IsValid (Actor|GetComponentbyClass _array_element "/MusicEncounter/Components/BP_MusicEncounterComponent.BP_MusicEncounterComponent_C")
              (:Is%20Valid
                (MusicEncounter|SetMusicEncounterInstanceTag(Message) (Variables|Default|GetAI%20Pawn) (Variables|/MusicEncounter/Components/BP_MusicEncounterComponent.BP_MusicEncounterComponent_C|GetMusic%20Encounter%20Instance%20Tag (Actor|GetComponentbyClass _array_element "/MusicEncounter/Components/BP_MusicEncounterComponent.BP_MusicEncounterComponent_C")))))))))))

(fn KnockbackPawns
  (params (Pawn :type Fort%20Pawn%20Object%20Reference))
  (bind (_returnvalue_x _returnvalue_y) (Math|Vector|BreakVector (Math|Vector|Normalize (- (Transformation|GetActorLocation Pawn) (Transformation|GetActorLocation (Ability|GetAvatarActorfromActorInfo))))))
  (if (not (GameplayTags|HasMatchingGameplayTag Pawn "(TagName=\"Gameplay.Status.KnockbackImmunity\")"))
    (Character|LaunchCharacter Pawn (Math|Vector|MakeVector (* _returnvalue_x (ScalableFloat|GetValueatLevel (Variables|Default|GetKnockbackPawnForwardVelocity))) (* _returnvalue_y (ScalableFloat|GetValueatLevel (Variables|Default|GetKnockbackPawnForwardVelocity))) (ScalableFloat|GetValueatLevel (Variables|Default|GetKnockbackPawnHeightVelocity))) true true)))

(fn KnockbackVehicles
  (params (FortAthenaVehicle :type Fort%20Athena%20Vehicle%20Object%20Reference))
  (bind (_returnvalue_x _returnvalue_y) (Math|Vector|BreakVector (Math|Vector|Normalize (- (Transformation|GetActorLocation FortAthenaVehicle) (Transformation|GetActorLocation (Ability|GetAvatarActorfromActorInfo))))))
  (Audio|BroadcastAbilityImpactatLocation 0 0 (Math|Vector|MakeVector (* _returnvalue_x (ScalableFloat|GetValueatLevel (Variables|Default|GetKnockbackVehicleForwardVelocity))) (* _returnvalue_y (ScalableFloat|GetValueatLevel (Variables|Default|GetKnockbackVehicleForwardVelocity))) (ScalableFloat|GetValueatLevel (Variables|Default|GetKnockbackVehicleHeightVelocity))) 777.0 "None" false))
